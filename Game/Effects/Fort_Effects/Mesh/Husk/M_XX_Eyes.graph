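# Material for husk "XX" eye meshes, usable on Cascade and Niagara mesh particles with Default Lit shading.
# Particle color RGB feeds the emissive input of MF_FX_EmissiveAdjustments, so the particle system drives the glow color.
# ConvertFromDiffSpec takes a default constant as diffuse color, with no textures in the graph.
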Material /Game/Effects/Fort_Effects/Mesh/Husk/M_XX_Eyes
graph {
 properties {bUsedWithMeshParticles=true, bUsedWithNiagaraMeshParticles=true, pixelDepthOffsetMode=PDOM_Legacy, usedShadingModels=Default Lit}
 nodes [
  MaterialFunctionCall(Emissive=ParticleColor().RGB) {materialFunction=&MF_FX_EmissiveAdjustments}
  MaterialFunctionCall(DiffuseColor=Constant()) {materialFunction=&ConvertFromDiffSpec}
 ]
}
